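; BP_CopyCollision is a Geometry Script tool actor that builds collision for a target static mesh from a separate collision mesh.
; On generated mesh rebuild, it stores the dynamic mesh and copies CollisionMesh into it.
; CopyMeshCollision writes convex-hull collision from that dynamic mesh onto StaticMeshTarget, marked as customized; nothing in this graph calls it.
; The construction script assigns StaticMeshTarget to the static mesh component; the play-time events are empty.
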(event EventBeginPlay)

(event Collision|EventActorBeginOverlap
  (params (OtherActor :type Actor%20Object%20Reference)))

(event EventTick
  (params (DeltaSeconds :type Float%20%28single-precision%29)))

(event Events|EventOnRebuildGeneratedMesh
  (params (TargetMesh :type Dynamic%20Mesh%20Object%20Reference))
  (bind _output_get (Variables|Default|SetDynamic%20Mesh TargetMesh))
  (GeometryScript|StaticMesh|CopyMeshFromStaticMesh (Variables|Default|GetCollisionMesh) _output_get))

(fn CopyMeshCollision
  (GeometryScript|Collision|SetStaticMeshCollisionfromMesh (Variables|Default|GetDynamic%20Mesh) (Variables|Default|GetStaticMeshTarget) (Utilities|Struct|MakeGeometryScriptCollisionfromMeshOptions true "ConvexHulls") "(bMarkAsCustomized=True)"))

(fn ConstructionScript
  (Components|StaticMesh|SetStaticMesh (Variables|Default|GetStaticMesh) (Variables|Default|GetStaticMeshTarget)))
